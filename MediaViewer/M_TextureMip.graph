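# M_TextureMip is a translucent UI material in MediaViewer for sampling a texture at a chosen mip level.
# A Texture object parameter feeds a texture sample set to explicit mip-level mode, with the level driven by the MipLevel scalar parameter.
# The sampler uses shared clamped world-group settings, and the graph shows no links from the sample to material outputs.
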Material /MediaViewer/M_TextureMip
graph {
 properties {blendMode=BLEND_Translucent, materialDomain=MD_UI, usedShadingModels=Default Lit}
 nodes [
  n1 = TextureObjectParameter!(UVs, Level, Apply View MipBias) {mipValueMode=TMVM_MipLevel, parameterName=Texture, samplerSource=SSM_Clamp_WorldGroupSettings, texture=/Engine/EngineResources/AICON-Green}
  TextureSample!(UVs, Tex=@n1, Level=ScalarParameter() {parameterName=MipLevel}, Apply View MipBias=StaticBool() {value=true}) {mipValueMode=TMVM_MipLevel, samplerSource=SSM_Clamp_WorldGroupSettings}
 ]
}
